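# Spot ambience cue for a power tower, layering wind-vibrated metal with intermittent creaks and groans.
# A randomly picked wind loop runs through nested envelopers that fade it in and add slow looping pitch and volume swells.
# It mixes with a quieter Looping branch that plays random, pitch-modulated metal stress one-shots after random delays.
# The mix fills only the first two quality-level slots, and a concurrency override allows one instance.
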SoundCue /Game/Sounds/Asteria/Environment_Actors/PowerTower/Spot_PowerTower_Wind_Through_Metal_Creaking_Cue
root=Delay_0
nodes [
 {id=Delay_0, children [QualityLevel_0], properties {delayMax=0.1, delayMin=0.05}}
 {id=Delay_2, children [Modulator_2], properties {delayMax=10, delayMin=3}}
 {
  id=Enveloper_0
  children [Modulator_1]
  properties {
   bLoop=true
   bLoopIndefinitely=true
   loopEnd=30
   pitchCurve {editorCurveData {keys [{interpMode=RCIM_Cubic, time=0}, <at 15.9468 0.9 RCIM_Cubic>, {interpMode=RCIM_Cubic, time=30}]}}
   volumeCurve {editorCurveData {keys [<at 0 1>]}}
  }
 }
 {
  id=Enveloper_29
  children [Enveloper_3]
  properties {
   pitchCurve {editorCurveData {keys [{time=0}]}}
   volumeCurve {editorCurveData {keys [{arriveTangent=3.4293, interpMode=RCIM_Cubic, leaveTangent=3.4293, tangentMode=RCTM_User, time=0, value=0}, <at 0.4 1 RCIM_Cubic>]}}
  }
 }
 {
  id=Enveloper_3
  children [Enveloper_0]
  properties {
   bLoop=true
   bLoopIndefinitely=true
   loopEnd=120
   pitchCurve {editorCurveData {keys [{time=0}]}}
   volumeCurve {editorCurveData {keys [
      {arriveTangent=-0.01, interpMode=RCIM_Cubic, time=0, value=1}
      <at 10.4789 0.5051 RCIM_Cubic>
      {arriveTangent=-0.01, interpMode=RCIM_Cubic, leaveTangent=-0.01, time=20, value=1}
      <at 40.4789 0.2051 RCIM_Cubic>
      {arriveTangent=0.002, interpMode=RCIM_Cubic, leaveTangent=0.002, time=60, value=1}
      <at 90.4789 0.3051 RCIM_Cubic>
      <at 120 1>
     ]}}
  }
 }
 {id=Looping_0, children [Delay_2]}
 {id=Mixer_0, children [WavePlayer_0], properties {inputVolume [1]}}
 {id=Mixer_1, children [WavePlayer_1], properties {inputVolume [1]}}
 {id=Mixer_2, children [Enveloper_29,Looping_0], properties {inputVolume [0.8,0.2]}}
 {id=Modulator_1, children [Random_0], properties {pitchMax=1.2, pitchMin=1, volumeMax=0.8, volumeMin=0.7}}
 {id=Modulator_2, children [Random_1], properties {pitchMax=1.2, pitchMin=1}}
 {id=QualityLevel_0, children [Mixer_2,Mixer_2,null,null]}
 {id=Random_0, children [Mixer_0,Mixer_1], properties {weights [1,1]}}
 {
  id=Random_1
  children [
   WavePlayer_2
   WavePlayer_3
   WavePlayer_4
   WavePlayer_5
   WavePlayer_6
   WavePlayer_13
   WavePlayer_14
   WavePlayer_15
   WavePlayer_16
   WavePlayer_7
   WavePlayer_8
   WavePlayer_9
  ]
  properties {weights [1,1,1,1,1,1,1,1,1,1,1,1]}
 }
 {
  id=WavePlayer_0
  properties {bLooping=true, soundWaveAssetPtr=/Game/Sounds/Artemis/POI/Blimps_Crashsites/Spots/Spot_Blimp_Crashsite_Wind_Vibrating_Metal_Large}
 }
 {
  id=WavePlayer_1
  properties {bLooping=true, soundWaveAssetPtr=/Game/Sounds/Artemis/POI/Blimps_Crashsites/Spots/Spot_Wind_Excited_Metal_Vibrating_Gusty_Loop}
 }
 {id=WavePlayer_13, properties {soundWaveAssetPtr=/Game/Sounds/Apollo/Biomes/Generic/AMB_Metal_Stress_Wronk_Groan_Knock_02}}
 {id=WavePlayer_14, properties {soundWaveAssetPtr=/Game/Sounds/Apollo/Biomes/Generic/AMB_Metal_Stress_Wronk_Groan_Stretch_02}}
 {id=WavePlayer_15, properties {soundWaveAssetPtr=/Game/Sounds/Apollo/Biomes/Generic/AMB_Metal_Stress_Wronk_Groan_Stretch_03}}
 {id=WavePlayer_16, properties {soundWaveAssetPtr=/Game/Sounds/Apollo/Biomes/Generic/AMB_Metal_Stress_Wronk_Groan_Stretch_04}}
 {id=WavePlayer_2, properties {soundWaveAssetPtr=/Game/Sounds/Apollo/Biomes/Generic/AMB_Metal_Stress_Wronk_Groan_Knock_01}}
 {id=WavePlayer_3, properties {soundWaveAssetPtr=/Game/Sounds/Apollo/Biomes/Generic/AMB_Metal_Stress_Wronk_Groan_Knock_06}}
 {id=WavePlayer_4, properties {soundWaveAssetPtr=/Game/Sounds/Apollo/Biomes/Generic/AMB_Metal_Stress_Wronk_Groan_Knock_05}}
 {id=WavePlayer_5, properties {soundWaveAssetPtr=/Game/Sounds/Apollo/Biomes/Generic/AMB_Metal_Stress_Wronk_Groan_Knock_04}}
 {id=WavePlayer_6, properties {soundWaveAssetPtr=/Game/Sounds/Apollo/Biomes/Generic/AMB_Metal_Stress_Wronk_Groan_Knock_03}}
 {id=WavePlayer_7, properties {soundWaveAssetPtr=/Game/Sounds/Apollo/Biomes/Generic/AMB_Metal_Stress_Wronk_Groan_Knock_09}}
 {id=WavePlayer_8, properties {soundWaveAssetPtr=/Game/Sounds/Apollo/Biomes/Generic/AMB_Metal_Stress_Wronk_Groan_Knock_07}}
 {id=WavePlayer_9, properties {soundWaveAssetPtr=/Game/Sounds/Apollo/Biomes/Generic/AMB_Metal_Stress_Wronk_Groan_Knock_08}}
]
properties {
 attenuationSettings=/Game/Sounds/Artemis/POI/Blimps_Crashsites/Spots/Spot_Wind_Vibrating_Metal_Structure_Large_Attenuation
 bEnableBusSends=false
 bEnableSubmixSends=false
 bOverrideConcurrency=true
 maxDistance=2140.1951
 soundClassObject=&Fort_Gadgets
 volumeMultiplier=1
 attenuationOverrides {attenuationShape=Capsule, falloffDistance=2223.3242, attenuationShapeExtents=<866.3708 230.6101>}
 concurrencyOverrides {maxCount=1, voiceStealReleaseTime=3}
}
